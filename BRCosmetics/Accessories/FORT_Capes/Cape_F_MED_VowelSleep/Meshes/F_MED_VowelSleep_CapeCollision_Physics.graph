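# Physics asset that supplies collision shapes for the F_MED_VowelSleep cape cosmetic, likely for its cloth or bone-driven secondary motion.
# Three capsules on consecutive dyn_hair_rear bones, narrowing along the chain, form an offset collision volume.
# A large flat box on dyn_ground_collision, with its top at the bone origin, acts as a floor.
# It has no constraints, so it defines collision volumes only, not a jointed ragdoll.
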PhysicsAsset /BRCosmetics/Accessories/FORT_Capes/Cape_F_MED_VowelSleep/Meshes/F_MED_VowelSleep_CapeCollision_Physics
bodies {
 dyn_ground_collision [{shapeType=Box, extentX=500, extentY=500, extentZ=80, shapeName=dyn_ground_collision_box, center=<0 0 -40>, rotation=<rot 0 -90 0>}]
 dyn_hair_rear_03 [{shapeType=Capsule, length=0.32, radius=11.4128, shapeName=dyn_hair_rear_03_capsule, center=<0.6839 -8.8102 0>, rotation=<rot 0 -90 0>}]
 dyn_hair_rear_04 [{shapeType=Capsule, length=1.5766, radius=10.5052, shapeName=dyn_hair_rear_04_capsule, center=<2.6634 -8.894 0>, rotation=<rot 0 -90 0>}]
 dyn_hair_rear_05 [{shapeType=Capsule, length=4.3629, radius=7.0697, shapeName=dyn_hair_rear_05_capsule, center=<6.7513 -6.7536 0>}]
}
constraints []
