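# Physics asset for the F_MED_VowelSleep back cape, defining the bodies and joints that simulate its swing.
# Four capsule chains hang from the left and right clavicle_cape bones and, via MSTR bones, backpack_root, with mostly wider swing and tighter twist limits lower down.
# Angularly free cross-links tie neighboring chains together mid-length.
# Capsules on torso, arms and legs provide collision shapes, and a huge root sphere approximates the ground.
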PhysicsAsset /BRCosmetics/Accessories/FORT_Capes/Cape_F_MED_VowelSleep/Meshes/F_MED_VowelSleep_Cape_Phys
bodies {
 backpack_root [{shapeType=Capsule, length=1, radius=1}]
 calf_l [{shapeType=Capsule, length=11.7963, radius=39.0787, center=<-26.1258 24.118 0>, rotation=<rot 90 179.8881 -178.6232>}]
 calf_r [{shapeType=Capsule, length=11.7963, radius=39.0787, center=<26.1258 -24.118 0>, rotation=<rot 90 180 1.4887>}]
 cape_01_l [{shapeType=Capsule, length=1, radius=1}]
 cape_01_r [{shapeType=Capsule, length=1, radius=1}]
 cape_02_l [{shapeType=Capsule, length=11.279, radius=9.0099, center=<9.0304 0 12.4752>, rotation=<rot 81.2746 0 0>}]
 cape_02_r [{shapeType=Capsule, length=12.5936, radius=8.4064, center=<9.0679 0 11.8758>, rotation=<rot 85.5914 0 0>}]
 cape_03_l [{shapeType=Capsule, length=11.7281, radius=12.2719, center=<11.9996 0 9.8467>, rotation=<rot 89.8477 0 0>}]
 cape_03_r [{shapeType=Capsule, length=11.2545, radius=12.7455, center=<12 0 10>, rotation=<rot 89.8477 0 0>}]
 cape_04_l [{shapeType=Capsule, length=12, radius=12.723, center=<8.8789 0 10.9427>, rotation=<rot 87.2914 0 0>}]
 cape_04_r [{shapeType=Capsule, length=12, radius=12.723, center=<9 0 13.502>, rotation=<rot 89.8477 0 0>}]
 cape_05_l [{shapeType=Capsule, length=12, radius=12.723, center=<8.9889 0 9.3164>, rotation=<rot 88.7143 0 0>}]
 cape_05_r [{shapeType=Capsule, length=12, radius=12.723, center=<9 0 13.502>, rotation=<rot 89.8477 0 0>}]
 cape_06_l [{shapeType=Capsule, length=12, radius=12.723, center=<9.0017 0 8.1438>, rotation=<rot 87.4409 0 0>}]
 cape_06_r [{shapeType=Capsule, length=12, radius=12.723, center=<9 0 13.502>, rotation=<rot 89.8477 0 0>}]
 cape_07_l [{shapeType=Capsule, length=12, radius=12.723, center=<8.9847 0 7.7473>, rotation=<rot 89.8477 0 0>}]
 cape_07_r [{shapeType=Capsule, length=12, radius=12.723, center=<9 0 13.502>, rotation=<rot 89.8477 0 0>}]
 cape_mid_01_l [{shapeType=Capsule, length=12, radius=9, center=<9.0043 0 11.6357>, rotation=<rot 89.8477 0 0>}]
 cape_mid_01_l_MSTR [{shapeType=Capsule, length=1, radius=1}]
 cape_mid_01_r [{shapeType=Capsule, length=12, radius=9, center=<9 0 10>, rotation=<rot 89.8477 0 0>}]
 cape_mid_01_r_MSTR [{shapeType=Capsule, length=1, radius=1}]
 cape_mid_02_l [{shapeType=Capsule, length=13.4725, radius=10.5275, center=<12 0 10>, rotation=<rot 89.8477 0 0>}]
 cape_mid_02_r [{shapeType=Capsule, length=13.3045, radius=10.6955, center=<12 0 10>, rotation=<rot 89.8477 0 0>}]
 cape_mid_03_l [{shapeType=Capsule, length=12, radius=12.723, center=<8.9936 0 11.0975>, rotation=<rot 88.7396 -0.0001 -0.0001>}]
 cape_mid_03_r [{shapeType=Capsule, length=12, radius=12.723, center=<9 0 13.502>, rotation=<rot 89.8477 0 0>}]
 cape_mid_04_l [{shapeType=Capsule, length=12, radius=12.723, center=<8.9899 0 9.6896>, rotation=<rot 88.4526 0 0>}]
 cape_mid_04_r [{shapeType=Capsule, length=12, radius=12.723, center=<9 0 13.502>, rotation=<rot 89.8477 0 0>}]
 cape_mid_05_l [{shapeType=Capsule, length=12, radius=12.723, center=<8.8936 -0.0024 9.678>, rotation=<rot 87.6489 1.3491 -0.0935>}]
 cape_mid_05_r [{shapeType=Capsule, length=12, radius=12.723, center=<9 0 13.502>, rotation=<rot 89.8477 0 0>}]
 cape_mid_06_l [{shapeType=Capsule, length=12, radius=12.723, center=<8.9462 0 9.3067>, rotation=<rot 89.2657 0 0>}]
 cape_mid_06_r [{shapeType=Capsule, length=12, radius=12.723, center=<9 0 13.502>, rotation=<rot 89.8477 0 0>}]
 clavicle_cape_l [{shapeType=Capsule, length=1, radius=1}]
 clavicle_cape_r [{shapeType=Capsule, length=1, radius=1}]
 clavicle_l [{shapeType=Capsule, length=14.1836, radius=8.4414, center=<12.1946 -0.1457 0>, rotation=<rot 90 -1.8948 0>}]
 clavicle_r [{shapeType=Capsule, length=14.1836, radius=8.4414, center=<-12.1946 0.1457 0>, rotation=<rot 90 -176.4464 0>}]
 lowerarm_l [{shapeType=Capsule, length=33.4684, radius=8.7868, shapeName=lowerarm_l_capsule, center=<18.6688 0 -1.2376>, rotation=<rot 0 -90 94.4658>}]
 lowerarm_r [{shapeType=Capsule, length=33.4684, radius=8.7868, shapeName=lowerarm_r_capsule, center=<-18.6688 0 1.2376>, rotation=<rot 0 -90 -85.5342>}]
 pelvis [{shapeType=Capsule, length=3.4412, radius=14.6526, center=<-4.178 -0.9183 0.03>}]
 root [{shapeType=Sphere, radius=700, shapeName=root_sphere, center=<0 0 -695>}]
 spine_02 [{shapeType=Capsule, length=2.377, radius=13.4355, center=<3.3885 1.2066 0>, rotation=<rot 0 8.071 0>}]
 spine_04 [{shapeType=Capsule, length=20.8161, radius=12.8071, center=<3.2944 2.0331 0.1009>, rotation=<rot 0 11.9036 9.8323>}]
 thigh_l [{shapeType=Capsule, length=28.0938, radius=9.2188, center=<-19.7778 -0.1788 0>, rotation=<rot 90 18.9704 16.3433>}]
 thigh_r [{shapeType=Capsule, length=28.0938, radius=9.2188, center=<19.7778 0.1788 0>, rotation=<rot 90 -10.6197 166.7532>}]
 upperarm_l [{shapeType=Capsule, length=19.4572, radius=8.1053, center=<11.9342 -1.3212 0>, rotation=<rot 89.8066 5.5324 8.8754>}]
 upperarm_r [{shapeType=Capsule, length=20.625, radius=6.9375, center=<-11.9342 1.3212 0>, rotation=<rot 90 -177.7987 0>}]
}
constraints [
 {
  bone1Name=cape_mid_01_l_MSTR
  bone2Name=backpack_root
  swing1LimitDegrees=45
  swing1Motion=Free
  swing2LimitDegrees=45
  swing2Motion=Free
  twistLimitDegrees=45
  twistMotion=Free
 }
 {bone1Name=cape_mid_01_l, bone2Name=cape_mid_01_l_MSTR, swing1LimitDegrees=5, swing2LimitDegrees=14, twistLimitDegrees=5}
 {
  bone1Name=cape_mid_01_r_MSTR
  bone2Name=backpack_root
  swing1LimitDegrees=45
  swing1Motion=Free
  swing2LimitDegrees=45
  swing2Motion=Free
  twistLimitDegrees=45
  twistMotion=Free
 }
 {bone1Name=cape_mid_01_r, bone2Name=cape_mid_01_r_MSTR, swing1LimitDegrees=5, swing2LimitDegrees=14, twistLimitDegrees=5}
 {bone1Name=cape_01_l, bone2Name=clavicle_cape_l, swing1LimitDegrees=45, swing2LimitDegrees=45, twistLimitDegrees=5}
 {bone1Name=cape_02_l, bone2Name=cape_01_l, swing1LimitDegrees=5, swing2LimitDegrees=14, twistLimitDegrees=5}
 {bone1Name=cape_03_l, bone2Name=cape_02_l, swing1LimitDegrees=5, swing2LimitDegrees=14, twistLimitDegrees=5}
 {bone1Name=cape_04_l, bone2Name=cape_03_l, swing1LimitDegrees=10, swing2LimitDegrees=24, twistLimitDegrees=1}
 {bone1Name=cape_05_l, bone2Name=cape_04_l, swing1LimitDegrees=10, swing2LimitDegrees=24, twistLimitDegrees=1}
 {bone1Name=cape_06_l, bone2Name=cape_05_l, swing1LimitDegrees=10, swing2LimitDegrees=24, twistLimitDegrees=1}
 {bone1Name=cape_07_l, bone2Name=cape_06_l, swing1LimitDegrees=10, swing2LimitDegrees=24, twistLimitDegrees=1}
 {bone1Name=cape_01_r, bone2Name=clavicle_cape_r, swing1LimitDegrees=45, swing2LimitDegrees=45, twistLimitDegrees=5}
 {bone1Name=cape_02_r, bone2Name=cape_01_r, swing1LimitDegrees=5, swing2LimitDegrees=14, twistLimitDegrees=5}
 {bone1Name=cape_03_r, bone2Name=cape_02_r, swing1LimitDegrees=5, swing2LimitDegrees=14, twistLimitDegrees=5}
 {bone1Name=cape_04_r, bone2Name=cape_03_r, swing1LimitDegrees=10, swing2LimitDegrees=24, twistLimitDegrees=1}
 {bone1Name=cape_05_r, bone2Name=cape_04_r, swing1LimitDegrees=10, swing2LimitDegrees=24, twistLimitDegrees=1}
 {bone1Name=cape_06_r, bone2Name=cape_05_r, swing1LimitDegrees=10, swing2LimitDegrees=24, twistLimitDegrees=1}
 {bone1Name=cape_07_r, bone2Name=cape_06_r, swing1LimitDegrees=10, swing2LimitDegrees=24, twistLimitDegrees=1}
 {bone1Name=cape_mid_02_l, bone2Name=cape_mid_01_l, swing1LimitDegrees=5, swing2LimitDegrees=14, twistLimitDegrees=5}
 {bone1Name=cape_mid_03_l, bone2Name=cape_mid_02_l, swing1LimitDegrees=10, swing2LimitDegrees=24, twistLimitDegrees=1}
 {bone1Name=cape_mid_04_l, bone2Name=cape_mid_03_l, swing1LimitDegrees=10, swing2LimitDegrees=24, twistLimitDegrees=1}
 {bone1Name=cape_mid_05_l, bone2Name=cape_mid_04_l, swing1LimitDegrees=10, swing2LimitDegrees=24, twistLimitDegrees=1}
 {bone1Name=cape_mid_06_l, bone2Name=cape_mid_05_l, swing1LimitDegrees=10, swing2LimitDegrees=24, twistLimitDegrees=1}
 {bone1Name=cape_mid_02_r, bone2Name=cape_mid_01_r, swing1LimitDegrees=5, swing2LimitDegrees=14, twistLimitDegrees=5}
 {bone1Name=cape_mid_03_r, bone2Name=cape_mid_02_r, swing1LimitDegrees=10, swing2LimitDegrees=24, twistLimitDegrees=1}
 {bone1Name=cape_mid_04_r, bone2Name=cape_mid_03_r, swing1LimitDegrees=10, swing2LimitDegrees=24, twistLimitDegrees=1}
 {bone1Name=cape_mid_05_r, bone2Name=cape_mid_04_r, swing1LimitDegrees=10, swing2LimitDegrees=24, twistLimitDegrees=1}
 {bone1Name=cape_mid_06_r, bone2Name=cape_mid_05_r, swing1LimitDegrees=10, swing2LimitDegrees=24, twistLimitDegrees=1}
 {
  bone1Name=cape_mid_04_l
  bone2Name=cape_mid_04_r
  swing1LimitDegrees=45
  swing1Motion=Free
  swing2LimitDegrees=45
  swing2Motion=Free
  twistLimitDegrees=45
  twistMotion=Free
 }
 {
  bone1Name=cape_05_l
  bone2Name=cape_mid_04_l
  swing1LimitDegrees=45
  swing1Motion=Free
  swing2LimitDegrees=45
  swing2Motion=Free
  twistLimitDegrees=45
  twistMotion=Free
 }
 {
  bone1Name=cape_mid_04_r
  bone2Name=cape_05_r
  swing1LimitDegrees=45
  swing1Motion=Free
  swing2LimitDegrees=45
  swing2Motion=Free
  twistLimitDegrees=45
  twistMotion=Free
 }
 {
  bone1Name=clavicle_cape_l
  bone2Name=root
  swing1LimitDegrees=45
  swing1Motion=Free
  swing2LimitDegrees=45
  swing2Motion=Free
  twistLimitDegrees=45
  twistMotion=Free
 }
]
